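# Translucent unlit Niagara sprite material that draws an animated, wavy glow around the Fairy Mochi pickaxe blade.
# UV and polar math build procedural capsule gradients, apparently corrected for view-dependent particle stretch.
# Sine waves, with amplitude blended from idle to swing by SwingAlpha, distort them into glow, outline and scrolling squiggle masks.
# Particle color tints the layers, higher quality levels add depth fade, and a directional sprite function handles orientation and camera offset.
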Material /BRCosmetics/Weapons/FORT_Melee/Pickaxe_FairyMochi/FX/Materials/M_Pickaxe_FairyMochi_BladeGlow
graph {
 properties {
  bAutomaticallySetUsageInEditor=false
  bUsedWithNiagaraSprites=true
  blendMode=BLEND_Translucent
  shadingModel=MSM_Unlit
  usedShadingModels=Unlit
 }
 nodes [
  n1 = Time()
  n2 = ScalarParameter() {defaultValue=1, parameterName=Wave Speed}
  n3 = Multiply(A=@n1, B=@n2)
  n4 = LinearInterpolate(
   A=ScalarParameter() {defaultValue=0.1, parameterName=Wave Amplitude Idle}
   B=ScalarParameter() {defaultValue=0.2, parameterName=Wave Amplitude Swing}
   Alpha=ScalarParameter() {parameterName=SwingAlpha}
  )
  n5 = TextureCoordinate()
  n6 = Sign(None=Subtract(A=ComponentMask(None=@n5) {r=true}) {constB=0.5})
  n7 = NamedRerouteDeclaration(None=Multiply(A=Saturate(None=@n6), B=ScalarParameter() {defaultValue=0.5, parameterName=Wave Time Offset})) {name=Time Offset, nodeColor=<rgba 0.6172 0 1 1>}
  n8 = Subtract(
   A=Multiply(A=NamedRerouteUsage(), B=ScalarParameter() {defaultValue=5, parameterName=Wave Frequency})
   B=Multiply(A=Add(A=@n1, B=@n7), B=@n2)
  )
  n9 = LinearInterpolate(A=Multiply(A=Sine(None=@n3), B=@n4), B=Multiply(A=Sine(None=@n8), B=@n4), Alpha=NamedRerouteUsage())
  n10 = Add(A=Subtract(A=NamedRerouteUsage(), B=ScalarParameter() {defaultValue=0.5, parameterName=Glow Width}), B=@n9)
  n11 = Saturate(None=Divide(A=@n10, B=ScalarParameter() {defaultValue=0.075, parameterName=Outline Thickness}))
  NamedRerouteDeclaration(None=OneMinus(None=Round(None=@n11))) {name=Glow Mask, nodeColor=<rgba 0 0.5078 1 1>}
  n12 = TextureCoordinate()
  n13 = Abs(None=ConstantBiasScale(None=ComponentMask(None=@n12) {r=true}) {bias=-0.5, scale=2})
  n14 = ComponentMask(None=@n12) {g=true}
  n15 = NamedRerouteUsage()
  n16 = Divide(A=@n14, B=@n15)
  n17 = Divide(A=OneMinus(None=@n14), B=@n15)
  n18 = ConstantBiasScale(None=Min(A=Saturate(None=@n16), B=Saturate(None=@n17))) {bias=-0.5, scale=2}
  n19 = MaterialFunctionCall!(XY=AppendVector(A=@n13, B=Abs(None=@n18))) {materialFunction=/Game/Accessories/FORT_Backpacks/Backpack_F_MED_Celestial_Glow/FX/Materials/MF_CartesianToPolar}
  n20 = Multiply(A=@n16) {constB=2}
  n21 = Multiply(A=@n17) {constB=2}
  n22 = Saturate(None=Multiply(A=Floor(None=@n20), B=Floor(None=@n21))) {desc=Center bit mask}
  n23 = OneMinus(None=@n22)
  n24 = Multiply(A=Multiply(A=@n19.Angle) {constB=4}, B=@n23)
  NamedRerouteDeclaration(None=OneMinus(None=Saturate(None=@n24))) {name=Poles Fade, nodeColor=<rgba 0 0.8125 1 1>}
  n25 = VectorParameter() {parameterName=Glow Color and Opacity, defaultValue=<rgba 0.5 0.302 0.1 0.5>}
  n26 = NamedRerouteUsage()
  n27 = LinearInterpolate(
   A=LinearInterpolate(
    A=Multiply(A=@n25.RGBA, B=@n26)
    B=VectorParameter() {parameterName=Ouline Color and Opacity, defaultValue=<rgba 1.6 0.688 0.16 1>}.RGBA
    Alpha=NamedRerouteUsage()
   )
   B=VectorParameter() {parameterName=Squiggle Color and Opacity, defaultValue=<rgba 1 0.6795 0.159 1>}.RGBA
   Alpha=NamedRerouteUsage()
  )
  n28 = ComponentMask(None=@n27) {b=true, g=true, r=true}
  n29 = ParticleColor()
  NamedRerouteDeclaration(None=MaterialFunctionCall(Emissive=Multiply(A=@n28, B=@n29.RGB)) {materialFunction=&MF_FX_EmissiveAdjustments}.Result) {name=Emissive Color, nodeColor=<rgba 0.6875 1 0 1>}
  n30 = Multiply(A=ComponentMask(None=@n27) {a=true}, B=@n29.A)
  n31 = DepthFade(Opacity=@n30, FadeDistance=ScalarParameter() {defaultValue=11, parameterName=Depth Fade Distance})
  NamedRerouteDeclaration(None=QualitySwitch(Default=@n30, Low=@n30, High=@n31, Medium=@n30, Epic=@n31)) {name=Opacity, nodeColor=<rgba 1 0 0.9609 1>}
  n32 = Divide(
   A=Subtract(A=@n10, B=ScalarParameter() {defaultValue=0.2, parameterName=Squiggle Offset})
   B=ScalarParameter() {defaultValue=0.15, parameterName=Squiggle Gradient Width}
  )
  n33 = Abs(None=ConstantBiasScale(None=Saturate(None=@n32)) {bias=-0.5, scale=2})
  n34 = Time()
  n35 = NamedRerouteUsage()
  n36 = Subtract(A=NamedRerouteUsage(), B=Multiply(A=Add(A=@n34, B=@n35), B=ScalarParameter() {defaultValue=0.5, parameterName=Squiggle Speed}))
  n37 = ScalarParameter() {defaultValue=3, parameterName=Squiggle Frequency}
  n38 = ConstantBiasScale(None=Frac(None=Multiply(A=@n36, B=@n37))) {bias=-0.5, scale=2}
  n39 = ComponentMask(None=TextureCoordinate()) {g=true}
  n40 = OneMinus(None=Abs(None=ConstantBiasScale(None=@n39) {bias=-0.5, scale=2}))
  n41 = ScalarParameter() {defaultValue=0.5, parameterName=Squiggle Length}
  n42 = Saturate(None=Divide(A=Abs(None=@n38), B=Multiply(A=@n40, B=@n41)))
  n43 = ScalarParameter() {defaultValue=6, parameterName=Squiggle Thickness Exponent}
  n44 = ScalarParameter() {defaultValue=0.5, parameterName=Squiggle Thickness}
  n45 = OneMinus(None=NamedRerouteUsage())
  n46 = Multiply(A=OneMinus(None=Power(Base=@n42, Exp=@n43)), B=Saturate(None=Subtract(A=@n44, B=@n45)))
  n47 = Subtract(A=Divide(A=OneMinus(None=@n33), B=OneMinus(None=@n46)))
  n48 = ScalarParameter() {defaultValue=0.25, parameterName=Squiggle Softness}
  NamedRerouteDeclaration(None=Saturate(None=Divide(A=@n47, B=@n48))) {name=Squiggle Mask, nodeColor=<rgba 0.5703 1 0 1>}
  NamedRerouteDeclaration(None=@n9) {name=aaa, nodeColor=<rgba 0 1 0.2734 1>}
  n49 = Frac(None=@n11)
  n50 = OneMinus(None=Abs(None=ConstantBiasScale(None=@n49) {bias=-0.5, scale=2}))
  n51 = ScalarParameter() {defaultValue=0.333, parameterName=Outline Softness}
  NamedRerouteDeclaration(None=Saturate(None=Divide(A=@n50, B=@n51))) {name=Outline Mask, nodeColor=<rgba 0.3594 1 0 1>}
  n52 = ParticleSize()
  n53 = ComponentMask(None=@n52) {r=true}
  n54 = Transform(None=Constant3Vector() {constant=<rgba 0 0 1 0>}) {transformSourceType=TRANSFORMSOURCE_Local}
  n55 = CameraPositionWS()
  n56 = ParticlePositionWS()
  n57 = DotProduct(A=@n54, B=Normalize(VectorInput=Subtract(A=@n55, B=@n56)))
  n58 = SquareRoot(None=OneMinus(None=Abs(None=@n57)))
  n59 = LinearInterpolate(A=@n53, B=ComponentMask(None=@n52) {g=true}, Alpha=@n58)
  NamedRerouteDeclaration(None=Divide(A=@n53, B=@n59)) {name=Particle Stretch Ratio, nodeColor=<rgba 0.1953 1 0 1>}
  n60 = ScalarParameter() {defaultValue=-33, parameterName=Camera Offset Front}
  n61 = ScalarParameter() {defaultValue=-11, parameterName=Camera Offset Side}
  NamedRerouteDeclaration(None=MaterialFunctionCall!(
    UV
    XY Scale=AppendVector(A=@n53, B=@n59)
    Pivot Offset
    Axis Locked=StaticBool()
    Direction Vector=@n54
    Center Location=ParticlePositionWS()
    Additional World Position Offset
    Normal
    Camera Offset=LinearInterpolate(A=@n60, B=@n61, Alpha=@n58)
   ) {materialFunction=/BRCosmetics/Weapons/FORT_Melee/Pickaxe_FairyMochi/FX/MF_SpriteDirectional}.World Position Offset) {name=WPO, nodeColor=<rgba 0 0.7187 1 1>}
  n62 = Multiply(A=@n19.Angle, B=Sign(None=Subtract(A=@n14) {constB=0.5}) {desc=Base/tip mask})
  n63 = Multiply(A=@n15) {constB=0.5}
  n64 = OneMinus(None=@n15)
  n65 = Saturate(None=Divide(A=Subtract(A=@n14, B=@n63), B=@n64))
  NamedRerouteDeclaration(None=Add(A=Multiply(A=@n62, B=@n23), B=Multiply(A=@n65, B=@n64) {desc=One bit of the UV})) {name=Wave Gradient, nodeColor=<rgba 0.3359 1 0 1>}
  NamedRerouteDeclaration(None=LinearInterpolate(A=@n19.Radius, B=@n13, Alpha=@n22)) {name=Width Gradient, nodeColor=<rgba 1 0.6094 0 1>}
  NamedRerouteUsage()
  NamedRerouteUsage()
  NamedRerouteUsage()
  NamedRerouteUsage()
  NamedRerouteUsage()
  NamedRerouteUsage()
 ]
}
