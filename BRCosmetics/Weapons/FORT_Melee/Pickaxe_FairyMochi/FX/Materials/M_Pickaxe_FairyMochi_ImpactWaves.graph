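# Translucent unlit Niagara sprite material for the FairyMochi pickaxe impact effect, drawing a tapered beam with wavy edges and animated squiggles.
# Time-driven sines distort the width gradient, phase-offset per side, yielding glow, outline and squiggle masks that blend three colors.
# Particle color, an ends fade and lifetime-driven erosion shape the output, with depth fade except on Low quality.
# A separate polar-UV and flipbook branch appears unconnected to the named outputs.
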Material /BRCosmetics/Weapons/FORT_Melee/Pickaxe_FairyMochi/FX/Materials/M_Pickaxe_FairyMochi_ImpactWaves
graph {
 properties {
  bAutomaticallySetUsageInEditor=false
  bUsedWithNiagaraSprites=true
  bUsedWithStaticMesh=false
  blendMode=BLEND_Translucent
  shadingModel=MSM_Unlit
  twoSided=true
  usedShadingModels=Unlit
 }
 nodes [
  ComponentMask(None=TextureCoordinate()) {g=true}
  n1 = TextureCoordinate()
  n2 = ComponentMask(None=@n1) {g=true}
  n3 = MaterialFunctionCall!(UV=AppendVector(A=ComponentMask(None=@n1) {r=true}, B=Power(Base=@n2) {constExponent=1}), Twist Center, Twist Edges) {materialFunction=/Game/Accessories/FORT_Backpacks/Backpack_F_MED_Celestial_Glow/FX/Materials/MF_PolarUV}
  n4 = ComponentMask(None=@n3.Polar UVs) {r=true}
  n5 = VectorParameter() {parameterName=Angle Multipliers, defaultValue=<rgba 6.1875 8.9375 11 0>}
  MaterialExpressionMultiply_1 = Multiply(A=@n4, B=@n5.RGB)
  Convert(None=@MaterialExpressionMultiply_1) {
   convertInputs [{
     type=Vector3
     expressionInput {
      expression=&self:MaterialExpressionMultiply_1
      inputName=None
      mask=0
      maskA=0
      maskB=0
      maskG=0
      maskR=0
      outputIndex=0
     }
    }]
   convertOutputs [{type=Scalar}, {type=Scalar}, {type=Scalar}]
  }
  n6 = ComponentMask(None=@n3.Polar UVs) {g=true}
  n7 = ConstantBiasScale(None=Frac(None=Add(A=@n4) {constB=0.25})) {bias=-0.5, scale=2}
  n8 = DotProduct(A=@n7, B=@n7)
  n9 = Multiply(A=OneMinus(None=@n8)) {constB=0.5}
  n10 = Add(A=@n4, B=Multiply(A=OneMinus(None=@n9), B=@n5.RGB))
  n11 = Multiply(A=Time(), B=VectorParameter() {parameterName=Angle Speeds, defaultValue=<rgba -1 -2 -3 0>}.RGB)
  n12 = Multiply(A=Sine(None=Add(A=@n10, B=@n11)), B=VectorParameter() {parameterName=Angle Weights, defaultValue=<rgba 0.1 0.05 0.03 0>}.RGB)
  n13 = Power(Base=OneMinus(None=@n8)) {constExponent=1}
  n14 = Subtract(
   A=Add(
    A=ConstantBiasScale(None=@n6) {bias=-0.5, scale=2}
    B=MaterialFunctionCall(f3=@n12) {materialFunction=&AddComponents}.f3
   )
   B=LinearInterpolate(A=ScalarParameter() {parameterName=Offset Tip}, B=ScalarParameter() {defaultValue=0.5, parameterName=Offset Base}, Alpha=@n13)
  )
  n15 = ScalarParameter() {defaultValue=0.5, parameterName=DebugFlipbookFrame}
  n16 = Constant() {r=3}
  LinearInterpolate(
   A=Round(None=@n14)
   B=TextureSample(UVs=MaterialFunctionCall("Animation  Phase (0-1)"=@n15, Number of Rows=@n16, Number of Columns=@n16) {materialFunction=&FlipBook}.UVs) {texture=/BRCosmetics/Weapons/FORT_Melee/Pickaxe_LettuceCat/FX/T_2DToon_Flipbook_DirectionalSmoke}.RGB
  )
  n17 = ParticleColor()
  MaterialFunctionCall(Emissive=Multiply(A=VectorParameter() {parameterName=Color, defaultValue=<rgba 3 1.5 0 0>}.RGB, B=@n17.RGB)) {materialFunction=&MF_FX_EmissiveAdjustments}
  n18 = NamedRerouteUsage()
  n19 = NamedRerouteUsage()
  Multiply(A=@n18, B=@n19)
  n20 = NamedRerouteUsage()
  Multiply(A=@n20, B=@n19)
  Multiply(A=@n17.A)
  n21 = TextureCoordinate()
  n22 = NamedRerouteDeclaration(None=OneMinus(None=ComponentMask(None=@n21) {g=true})) {name=Wave Gradient, nodeColor=<rgba 1 0 0.5625 1>}
  n23 = Time()
  n24 = TextureCoordinate()
  n25 = Sign(None=Subtract(A=ComponentMask(None=@n24) {r=true}) {constB=0.5})
  n26 = NamedRerouteDeclaration(None=Multiply(A=Saturate(None=@n25), B=ScalarParameter() {defaultValue=0.5, parameterName=Wave Time Offset})) {name=Time Offset, nodeColor=<rgba 0.6172 0 1 1>}
  n27 = Subtract(
   A=Multiply(A=@n22, B=ScalarParameter() {defaultValue=5, parameterName=Wave Frequency})
   B=Multiply(A=Add(A=@n23, B=@n26), B=ScalarParameter() {defaultValue=2, parameterName=Wave Speed})
  )
  n28 = ScalarParameter() {defaultValue=0.1, parameterName=Wave Amplitude Idle}
  n29 = NamedRerouteUsage()
  n30 = Add(A=NamedRerouteUsage(), B=Multiply(A=Sine(None=@n27), B=Multiply(A=@n28, B=@n29)))
  n31 = Saturate(None=Divide(A=@n30, B=ScalarParameter() {defaultValue=0.05, parameterName=Outline Thickness}))
  NamedRerouteDeclaration(None=OneMinus(None=Round(None=@n31))) {name=Glow Mask, nodeColor=<rgba 0 0.5078 1 1>}
  n32 = ParticleColor()
  n33 = VectorParameter() {parameterName=Glow Color and Opacity, defaultValue=<rgba 0.5 0.302 0.1 0.5>}
  n34 = LinearInterpolate(
   A=LinearInterpolate(
    A=Multiply(A=@n33.RGBA, B=@n20)
    B=VectorParameter() {parameterName=Ouline Color and Opacity, defaultValue=<rgba 1.6 0.688 0.16 1>}.RGBA
    Alpha=@n18
   )
   B=VectorParameter() {parameterName=Squiggle Color and Opacity, defaultValue=<rgba 1 0.6795 0.159 1>}.RGBA
   Alpha=NamedRerouteUsage()
  )
  n35 = ComponentMask(None=@n34) {b=true, g=true, r=true}
  NamedRerouteDeclaration(None=MaterialFunctionCall(Emissive=Multiply(A=@n32.RGB, B=@n35)) {materialFunction=&MF_FX_EmissiveAdjustments}.Result) {name=Emissive Color, nodeColor=<rgba 0.7344 0 1 1>}
  n36 = ComponentMask(None=@n34) {a=true}
  n37 = NamedRerouteUsage()
  n38 = ParticleRelativeTime()
  n39 = ConstantBiasScale(None=@n38) {bias=-0.5, scale=2}
  n40 = OneMinus(None=@n38)
  n41 = ScalarParameter() {defaultValue=3, parameterName=Erosion Softness}
  n42 = Multiply(
   A=Multiply(A=Multiply(A=@n36, B=@n37), B=@n32.A)
   B=MaterialFunctionCall!(Gradient=NamedRerouteUsage(), Erosion Amount=Saturate(None=@n39), Softness=Multiply(A=@n40, B=@n41)) {materialFunction=/BRCosmetics/Athena/Effects/Particles/Emotes/NS_Emote_Spectrum/Functions/MF_Erosion}.Increased Range
  )
  n43 = DepthFade(Opacity=@n42, FadeDistance=ScalarParameter() {defaultValue=11, parameterName=Depth Fade Distance})
  NamedRerouteDeclaration(None=QualitySwitch(Default=@n42, Low=@n42, High=@n43, Medium=@n42, Epic=@n43)) {name=Opacity, nodeColor=<rgba 1 0 0.9609 1>}
  n44 = Divide(
   A=Subtract(A=@n30, B=ScalarParameter() {defaultValue=0.2, parameterName=Squiggle Offset})
   B=ScalarParameter() {defaultValue=0.15, parameterName=Squiggle Gradient Width}
  )
  n45 = Abs(None=ConstantBiasScale(None=Saturate(None=@n44)) {bias=-0.5, scale=2})
  n46 = Time()
  n47 = NamedRerouteUsage()
  n48 = Subtract(A=@n22, B=Multiply(A=Add(A=@n46, B=@n47), B=ScalarParameter() {defaultValue=1, parameterName=Squiggle Speed}))
  n49 = ScalarParameter() {defaultValue=3, parameterName=Squiggle Frequency}
  n50 = ConstantBiasScale(None=Frac(None=Multiply(A=@n48, B=@n49))) {bias=-0.5, scale=2}
  n51 = OneMinus(None=Abs(None=ConstantBiasScale(None=@n22) {bias=-0.5, scale=2}))
  n52 = ScalarParameter() {defaultValue=0.5, parameterName=Squiggle Length}
  n53 = Saturate(None=Divide(A=Abs(None=@n50), B=Multiply(A=@n51, B=@n52)))
  n54 = ScalarParameter() {defaultValue=6, parameterName=Squiggle Thickness Exponent}
  n55 = Multiply(A=OneMinus(None=Power(Base=@n53, Exp=@n54)), B=Saturate(None=ScalarParameter() {defaultValue=0.5, parameterName=Squiggle Thickness}))
  n56 = Subtract(A=Divide(A=OneMinus(None=@n45), B=OneMinus(None=@n55)))
  n57 = ScalarParameter() {defaultValue=0.25, parameterName=Squiggle Softness}
  NamedRerouteDeclaration(None=Saturate(None=Divide(A=@n56, B=@n57))) {name=Squiggle Mask, nodeColor=<rgba 0.5703 1 0 1>}
  n58 = TextureCoordinate()
  n59 = ConstantBiasScale(None=ComponentMask(None=@n58) {r=true}) {bias=-0.5, scale=2}
  n60 = ScalarParameter() {defaultValue=0.7, parameterName=Beam Width Tip}
  n61 = ScalarParameter() {defaultValue=-0.1, parameterName=Beam Width Base}
  n62 = ComponentMask(None=@n58) {g=true}
  NamedRerouteDeclaration(None=Subtract(A=Abs(None=@n59), B=LinearInterpolate(A=@n60, B=@n61, Alpha=@n62))) {name=Width Gradient, nodeColor=<rgba 0.2422 1 0 1>}
  n63 = Frac(None=@n31)
  n64 = OneMinus(None=Abs(None=ConstantBiasScale(None=@n63) {bias=-0.5, scale=2}))
  n65 = ScalarParameter() {defaultValue=0.333, parameterName=Outline Softness}
  NamedRerouteDeclaration(None=Saturate(None=Divide(A=@n64, B=@n65))) {name=Outline Mask, nodeColor=<rgba 0.3594 1 0 1>}
  n66 = ConstantBiasScale(None=NamedRerouteUsage()) {bias=-0.5, scale=2}
  n67 = OneMinus(None=Power(Base=DotProduct(A=@n66, B=@n66), Exp=ScalarParameter() {defaultValue=0.5, parameterName=Ends Fade Center Exponent}))
  NamedRerouteDeclaration(None=Power(Base=Saturate(None=@n67), Exp=ScalarParameter() {defaultValue=2, parameterName=Ends Fade Tip Exponent})) {name=Ends Mask, nodeColor=<rgba 1 0 0.6797 1>}
  NamedRerouteUsage()
  NamedRerouteUsage()
  n68 = ScalarParameter() {defaultValue=0.05, parameterName=Thickness Tip}
  n69 = ScalarParameter() {defaultValue=0.13, parameterName=Thickness Base}
  n70 = Saturate(None=Divide(A=@n14, B=LinearInterpolate(A=@n68, B=@n69, Alpha=@n13)))
  Saturate(None=Ceil(None=Frac(None=@n70)))
 ]
}
