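# Translucent unlit Niagara sprite material that draws animated squiggly spark streaks for the FairyMochi pickaxe effects.
# A time-scrolled sine along V, offset by particle random, shifts U to form a wavy line whose thickness tapers at both ends.
# That procedural mask, with adjustable softness, multiplies particle alpha, while particle color feeds the shared emissive adjustment function.
# A position-based wiggle offset and several position nodes appear unconnected.
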Material /BRCosmetics/Weapons/FORT_Melee/Pickaxe_FairyMochi/FX/Materials/M_Pickaxe_FairyMochi_SquigglySparks
graph {
 properties {
  bAutomaticallySetUsageInEditor=false
  bUsedWithNiagaraSprites=true
  blendMode=BLEND_Translucent
  shadingModel=MSM_Unlit
  usedShadingModels=Unlit
 }
 nodes [
  n1 = TextureCoordinate()
  n2 = ComponentMask(None=@n1) {g=true}
  n3 = Multiply(A=@n2, B=ScalarParameter() {defaultValue=1.62, parameterName=Squiggly Wiggles})
  n4 = ParticlePositionWS()
  n5 = Abs(None=TransformPosition(None=@n4) {transformSourceType=TRANSFORMPOSSOURCE_World})
  Add(
   A=@n3
   B=Divide(
    A=MaterialFunctionCall(Vector 3=@n5) {materialFunction=/Engine/Functions/Engine_MaterialFunctions02/Utility/VectorLength}.V3 Length
    B=ScalarParameter() {defaultValue=5, parameterName=Wiggle Distance Tile Size}
   )
  )
  LocalPosition()
  n6 = ParticleColor()
  MaterialFunctionCall(Emissive=@n6.RGB) {materialFunction=&MF_FX_EmissiveAdjustments}
  n7 = ComponentMask(None=@n1) {r=true}
  n8 = Add(A=Multiply(A=Time(), B=ScalarParameter() {defaultValue=4, parameterName=Wiggle Speed}), B=ParticleRandom())
  n9 = Multiply(A=Sine(None=Subtract(A=@n3, B=@n8)), B=ScalarParameter() {defaultValue=0.2, parameterName=Wiggle Intensity})
  n10 = Abs(None=ConstantBiasScale(None=Add(A=@n7, B=@n9)) {bias=-0.5, scale=2})
  n11 = Abs(None=ConstantBiasScale(None=@n2) {bias=-0.5, scale=2})
  n12 = Divide(
   A=Subtract(A=OneMinus(None=@n11), B=ScalarParameter() {defaultValue=0.3, parameterName=Squiggle Ends Fade Offset})
   B=ScalarParameter() {defaultValue=0.2, parameterName=Squiggle Ends Fade Size}
  )
  n13 = Multiply(A=Saturate(None=SquareRoot(None=@n12)), B=ScalarParameter() {defaultValue=0.2, parameterName=Squiggle Thickness Max})
  n14 = OneMinus(None=Saturate(None=Divide(A=@n10, B=@n13)))
  n15 = ScalarParameter() {defaultValue=0.3, parameterName=Squiggle Softness}
  Multiply(A=@n6.A, B=Saturate(None=Divide(A=@n14, B=@n15)))
  Transform(None=@n4)
  WorldPosition()
 ]
}
