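# Translucent unlit Niagara ribbon material for the FairyMochi pickaxe's swing trail, drawn procedurally from UVs without textures.
# A time-scrolled sine distorts the trail's edge, with UV-based parameters separating top and bottom wave, color and time offset.
# The wavy edge derives glow, outline and scrolling dashed squiggle masks, shaped by length and tip fades.
# These masks blend colors, which particle color multiplies into opacity and, via MF_FX_EmissiveAdjustments, emissive.
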Material /BRCosmetics/Weapons/FORT_Melee/Pickaxe_FairyMochi/FX/Materials/M_Pickaxe_FairyMochi_SwingTrail
graph {
 properties {
  bAutomaticallySetUsageInEditor=false
  bUsedWithNiagaraRibbons=true
  blendMode=BLEND_Translucent
  shadingModel=MSM_Unlit
  usedShadingModels=Unlit
 }
 nodes [
  n1 = TextureCoordinate() {coordinateIndex=1}
  n2 = NamedRerouteDeclaration(None=OneMinus(None=ComponentMask(None=@n1) {r=true})) {name=Wave Gradient, nodeColor=<rgba 1 0 0.5625 1>}
  n3 = Multiply(
   A=@n2
   B=LinearInterpolate(
    A=ScalarParameter() {defaultValue=5, parameterName=Wave Frequency Bottom}
    B=ScalarParameter() {defaultValue=7, parameterName=Wave Frequency Top}
    Alpha=NamedRerouteUsage()
   )
  )
  n4 = Multiply(A=NamedRerouteUsage(), B=ScalarParameter() {defaultValue=-2, parameterName=Wave Speed})
  n5 = LinearInterpolate(
   A=ScalarParameter() {defaultValue=0.3, parameterName=Wave Amplitude Start}
   B=ScalarParameter() {defaultValue=0.05, parameterName=Wave Amplitude End}
   Alpha=NamedRerouteUsage()
  )
  n6 = LinearInterpolate(A=ScalarParameter() {defaultValue=1.5, parameterName=Wave Amplitude Multiplier Bottom}, Alpha=NamedRerouteUsage())
  n7 = OneMinus(None=NamedRerouteUsage())
  n8 = Multiply(A=Sine(None=Subtract(A=@n3, B=@n4)), B=Multiply(A=Multiply(A=@n5, B=@n6), B=SquareRoot(None=@n7)))
  n9 = Add(A=NamedRerouteUsage(), B=@n8)
  Abs(None=@n9)
  n10 = TextureCoordinate() {coordinateIndex=1}
  n11 = Abs(None=ConstantBiasScale(None=ComponentMask(None=@n10) {g=true}) {bias=-0.5, scale=2})
  n12 = LinearInterpolate(
   A=ScalarParameter() {defaultValue=0.8, parameterName=Width Start}
   B=ScalarParameter() {defaultValue=0.2, parameterName=Width End}
   Alpha=NamedRerouteUsage()
  )
  n13 = Divide(A=@n11, B=@n12)
  n14 = Subtract(A=OneMinus(None=@n13), B=@n8)
  n15 = OneMinus(None=@n8)
  n16 = Max(A=Saturate(None=Divide(A=@n14, B=@n15)), B=OneMinus(None=NamedRerouteUsage()))
  n17 = ScalarParameter() {defaultValue=0.4, parameterName=Bottom Side Softness}
  n18 = ScalarParameter() {defaultValue=0.05, parameterName=Outline Thickness Bottom}
  n19 = ScalarParameter() {defaultValue=0.04, parameterName=Outline Thickness Top}
  n20 = NamedRerouteUsage()
  n21 = Saturate(None=Divide(A=@n9, B=LinearInterpolate(A=@n18, B=@n19, Alpha=@n20)))
  n22 = OneMinus(None=Round(None=@n21))
  Min(A=Saturate(None=Divide(A=@n16, B=@n17) {constB=0.1}), B=@n22)
  n23 = ComponentMask(None=TextureCoordinate()) {g=true}
  n24 = Saturate(None=Sign(None=Subtract(A=@n23) {constB=0.5}))
  NamedRerouteDeclaration(None=Multiply(
    A=NamedRerouteDeclaration(None=@n24) {name="Trail  Top Down Mask", nodeColor=<rgba 0.5703 0 1 1>}
    B=ScalarParameter() {parameterName=Wave Time Offset}
   )) {name=Time Offset, nodeColor=<rgba 0.6172 0 1 1>}
  n25 = Power(Base=OneMinus(None=NamedRerouteUsage()), Exp=ScalarParameter() {defaultValue=1, parameterName=Glow Length Fade Exponent})
  NamedRerouteDeclaration(None=Multiply(A=@n22, B=OneMinus(None=@n25))) {name=Glow Mask, nodeColor=<rgba 0 0.5078 1 1>}
  n26 = ParticleColor()
  n27 = LinearInterpolate(
   A=VectorParameter() {parameterName=Glow Color and Opacity Bottom, defaultValue=<rgba 0.3 0.1812 0.06 0.4>}.RGBA
   B=VectorParameter() {parameterName=Glow Color and Opacity, defaultValue=<rgba 0.55 0.3458 0.1375 0.6>}.RGBA
   Alpha=ComponentMask(None=TextureCoordinate()) {g=true}
  )
  n28 = NamedRerouteUsage()
  n29 = VectorParameter() {parameterName=Outline Color and Opacity Bottom, defaultValue=<rgba 1 0.6928 0.4 0.2>}
  n30 = VectorParameter() {parameterName=Ouline Color and Opacity, defaultValue=<rgba 1 0.575 0.15 1>}
  n31 = NamedRerouteUsage()
  n32 = LinearInterpolate(
   A=LinearInterpolate(A=Multiply(A=@n27, B=@n28), B=LinearInterpolate(A=@n29.RGBA, B=@n30.RGBA, Alpha=@n31), Alpha=NamedRerouteUsage())
   B=VectorParameter() {parameterName=Squiggle Color and Opacity, defaultValue=<rgba 0.6 0.4285 0.15 1>}.RGBA
   Alpha=NamedRerouteUsage()
  )
  n33 = ComponentMask(None=@n32) {b=true, g=true, r=true}
  NamedRerouteDeclaration(None=MaterialFunctionCall(Emissive=Multiply(A=@n26.RGB, B=@n33)) {materialFunction=&MF_FX_EmissiveAdjustments}.Result) {name=Emissive Color, nodeColor=<rgba 0.6406 0 1 1>}
  n34 = ComponentMask(None=TextureCoordinate()) {r=true}
  n35 = ScalarParameter() {defaultValue=1, parameterName=Locked Gradient Length}
  n36 = OneMinus(None=Saturate(None=Divide(A=@n34, B=@n35)))
  n37 = ScalarParameter() {defaultValue=1, parameterName=Length Fade Exponent}
  NamedRerouteDeclaration(None=OneMinus(None=Power(Base=@n36, Exp=@n37))) {name=Length Fade, nodeColor=<rgba 1 0 0.4687 1>}
  NamedRerouteDeclaration(None=Add(A=Time(), B=NamedRerouteUsage())) {name=Time w/ Offset, nodeColor=<rgba 0.7578 0 1 1>}
  n38 = Divide(
   A=Subtract(A=@n9, B=ScalarParameter() {defaultValue=0.0333, parameterName=Squiggle Offset})
   B=ScalarParameter() {defaultValue=0.2, parameterName=Squiggle Gradient Width}
  )
  n39 = Abs(None=ConstantBiasScale(None=Saturate(None=@n38)) {bias=-0.5, scale=2})
  n40 = Subtract(A=@n2, B=Multiply(A=NamedRerouteUsage(), B=ScalarParameter() {defaultValue=0.3, parameterName=Squiggle Speed}))
  n41 = LinearInterpolate(
   A=ScalarParameter() {parameterName=Squiggle Frequency Bottom}
   B=ScalarParameter() {defaultValue=6, parameterName=Squiggle Frequency Top}
   Alpha=NamedRerouteUsage()
  )
  n42 = ConstantBiasScale(None=Frac(None=Multiply(A=@n40, B=@n41))) {bias=-0.5, scale=2}
  n43 = ScalarParameter() {defaultValue=0.6, parameterName=Squiggle Length}
  n44 = NamedRerouteUsage()
  n45 = Saturate(None=Divide(A=Abs(None=@n42), B=Multiply(A=@n43, B=@n44)))
  n46 = ScalarParameter() {defaultValue=9, parameterName=Squiggle Thickness Exponent}
  n47 = ScalarParameter() {defaultValue=0.25, parameterName=Squiggle Thickness}
  n48 = Multiply(A=OneMinus(None=Power(Base=@n45, Exp=@n46)), B=Saturate(None=Multiply(A=@n47, B=@n44)))
  n49 = Subtract(A=Divide(A=OneMinus(None=@n39), B=OneMinus(None=@n48)))
  n50 = ScalarParameter() {defaultValue=0.2, parameterName=Squiggle Softness}
  NamedRerouteDeclaration(None=Saturate(None=Divide(A=@n49, B=@n50))) {name=Squiggle Mask, nodeColor=<rgba 0.5703 1 0 1>}
  NamedRerouteDeclaration(None=Subtract(A=@n11, B=@n12)) {name=Width Gradient, nodeColor=<rgba 1 0.9141 0 1>}
  n51 = ConstantBiasScale(None=Frac(None=@n21)) {bias=-0.5, scale=2}
  n52 = Divide(A=OneMinus(None=Abs(None=@n51)), B=Divide(A=ScalarParameter() {defaultValue=0.4, parameterName=Outline Softness}, B=NamedRerouteUsage()))
  n53 = Power(Base=OneMinus(None=NamedRerouteUsage()), Exp=ScalarParameter() {defaultValue=2, parameterName=Outline Length Fade Exponent})
  NamedRerouteDeclaration(None=Multiply(A=Saturate(None=@n52), B=OneMinus(None=@n53))) {name=Outline Mask, nodeColor=<rgba 0.3594 1 0 1>}
  NamedRerouteDeclaration(None=@n36) {name=Locked Gradient, nodeColor=<rgba 0 1 0.8359 1>}
  n54 = NamedRerouteUsage()
  n55 = OneMinus(None=@n54)
  n56 = Divide(A=@n55, B=ScalarParameter() {defaultValue=0.2, parameterName=Base Fade Size})
  n57 = Divide(A=@n54, B=ScalarParameter() {defaultValue=0.01, parameterName=End Fade Size})
  NamedRerouteDeclaration(None=Multiply(A=Saturate(None=@n56), B=Saturate(None=@n57))) {name=Tips Fade, nodeColor=<rgba 0.7109 0 1 1>}
  n58 = Saturate(None=Divide(A=@n55, B=ScalarParameter() {defaultValue=0.8, parameterName=Squiggle Range}))
  NamedRerouteDeclaration(None=OneMinus(None=Power(Base=@n58) {constExponent=3})) {name=Squiggle Range, nodeColor=<rgba 0 0.6719 1 1>}
  n59 = ComponentMask(None=@n32) {a=true}
  NamedRerouteDeclaration(None=Multiply(A=Multiply(A=@n59, B=@n26.A), B=NamedRerouteUsage())) {name=Opacity, nodeColor=<rgba 1 0.9609 0 1>}
  NamedRerouteUsage()
  Saturate(None=@n13)
 ]
}
