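# Idle Niagara effect for the FairyMochi pickaxe, made of a single sprite glow emitter, WigglyGlow.
# Position and a tall sprite size are set once at emitter spawn, with no particle spawn modules shown.
# Each update samples a color curve by the Fortnite time-of-day parameter collection value, tinting the glow dimmer toward night and brighter by day at full alpha.
# A User.bIsSwinging parameter is exposed, but no shown module reads it.
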NiagaraSystem /BRCosmetics/Weapons/FORT_Melee/Pickaxe_FairyMochi/FX/NS_Pickaxe_FairyMochi_Idle
dependencies {
 usedDataInterfaces [&NiagaraDataInterfaceColorCurve]
 usedDynamicInputs []
 usedModules [&SystemState, &Timeline]
}
emitters {WigglyGlow {
  emitterSpawnScript [/SetVariables(Emitter.Position:NiagaraPosition=<0 0 40>, Emitter.SpriteSize:Vector2f=<22 77>) {isSetParameters=true}]
  emitterUpdateScript [
   &EmitterState()
   /SetVariables(Emitter.Linear Color:LinearColor=&MakeLinearColorFromVectorAndFloat(
     "Vector (RGB)"=&MakeVectorFromLinearColorRGB(Color=&ColorFromCurve(
       ColorCurve=DataInterface() {propertyValues {
         AlphaCurve {keys [<at 0 1>,<at 0.125 1>,<at 0.25 1>,<at 0.375 1>,<at 0.5 1>,<at 0.625 1>,<at 0.75 1>,<at 0.875 1>,<at 1 1>]}
         BlueCurve {keys [<at 0.1279 0.2852>,<at 0.166 0.6172>,<at 0.208 0.5395>,<at 0.25 0.6172>,<at 0.291 0.6939>,<at 0.333 0.7379>,<at 0.375 0.7835>,<at 0.416 0.7991>,<at 0.458 0.8148>,<at 0.5 0.8308>,<at 0.541 0.8148>,<at 0.583 0.7991>,<at 0.625 0.7991>,<at 0.666 0.7529>,<at 0.708 0.7084>,<at 0.75 0.5711>,<at 0.791 0.4735>,<at 0.8274 0.3231>,<at 0.8673 0.2852>]}
         GreenCurve {keys [<at 0.1279 0.3249>,<at 0.166 0.5333>,<at 0.208 0.5271>,<at 0.25 0.6376>,<at 0.291 0.7529>,<at 0.333 0.7835>,<at 0.375 0.8148>,<at 0.416 0.7991>,<at 0.458 0.8148>,<at 0.5 0.8469>,<at 0.541 0.7991>,<at 0.583 0.7991>,<at 0.625 0.8148>,<at 0.666 0.7991>,<at 0.708 0.7529>,<at 0.75 0.5711>,<at 0.791 0.5029>,<at 0.8274 0.5029>,<at 0.8673 0.3249>]}
         RedCurve {keys [<at 0.1279 0.3802>,<at 0.166 0.7084>,<at 0.208 0.6867>,<at 0.25 0.8148>,<at 0.291 0.8879>,<at 0.333 0.8963>,<at 0.375 0.9131>,<at 0.416 0.8714>,<at 0.458 0.8714>,<at 0.5 0.9131>,<at 0.541 0.8632>,<at 0.583 0.8714>,<at 0.625 0.9131>,<at 0.666 0.8963>,<at 0.708 0.8714>,<at 0.75 0.7529>,<at 0.791 0.7529>,<at 0.8274 0.8469>,<at 0.8673 0.3802>]}
        }}
       CurveIndex=Linked() {linkedVariable {name=NPC.FortniteNPC.TimeOfDay_0-1, type=NiagaraFloat}}
      ))
     "Float (Alpha)"=1
    )) {isSetParameters=true}
  ]
  rendererClasses [&NiagaraSpriteRendererProperties]
  renderers [{rendererClass=&NiagaraSpriteRendererProperties}]
 }}
summary {
 systemName=NS_Pickaxe_FairyMochi_Idle
 emitters [{emitterName=WigglyGlow, rendererClasses [&NiagaraSpriteRendererProperties]}]
 userVariables [User.bIsSwinging {type=NiagaraBool}]
}
