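# Niagara system for the FairyMochi pickaxe's swing trail, with one ribbon emitter, Main_Trail.
# The anim-trail modules spawn, initialize and update short-lived ribbon particles, with a ribbon ID offset utility at emitter spawn.
# User.SwingSectionIndex selects the trail material length from a float array.
# Color comes from a system time-of-day value, while a curve on system age eases an initial brightness boost and SystemFade scales alpha.
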NiagaraSystem /BRCosmetics/Weapons/FORT_Melee/Pickaxe_FairyMochi/FX/NS_Pickaxe_FairyMochi_Trail
dependencies {
 usedDataInterfaces [
  &NiagaraDataInterfaceArrayFloat
  &NiagaraDataInterfaceCurve
  &NiagaraDataInterfaceSocketReader
  &NiagaraDataInterfaceColorCurve
 ]
 usedDynamicInputs [&FNTimeOfDayScalar, &ColorFromCurve]
 usedModules [
  &self:SystemSpawnScript.SystemScriptSource.SystemScriptGraph.NiagaraNodeAssignment_3.SetVariables_#1
  &SystemState
  &self:SystemSpawnScript.SystemScriptSource.SystemScriptGraph.NiagaraNodeAssignment_4.SetVariables_#2
  &self:SystemSpawnScript.SystemScriptSource.SystemScriptGraph.NiagaraNodeAssignment_1.SetVariables_#3
 ]
}
emitters {Main_Trail {
  emitterSpawnScript [&FN_RibbonIDOffsetUtility()]
  emitterUpdateScript [
   &EmitterState()
   &FN_SpawnAnimTrail()
   /SetVariables(Emitter.TrailMaterialLength:NiagaraFloat=&SelectFloatFromArray(Direct Array Index=$User.SwingSectionIndex, Float Selection Array=DataInterface() {propertyValues {FloatData [1,0.75,0.75,0.75]}})) {isSetParameters=true}
  ]
  particleSpawnScript [
   &FN_InitializeAnimTrailBoneIndecies()
   &FN_InitializeAnimTrail(AnimTrailWidth=1.25, Color=$System.TOD, Lifetime=0.3)
  ]
  particleUpdateScript [
   &FN_UpdateAnimTrail() {module=FN_UpdateAnimTrail001}
   &ScaleColor(
    Scale Alpha=$System.SystemFade
    Scale RGB=&VectorFromFloat(Value=&FloatFromCurve(FloatCurve=DataInterface() {propertyValues {Curve {keys [<at 0 3>,<at 0.1666 1>]}}}, CurveIndex=$System.Age))
    ScaleA=true
    ScaleRGB=true
   )
  ]
  rendererClasses [&NiagaraRibbonRendererProperties]
  renderers [{rendererClass=&NiagaraRibbonRendererProperties}]
 }}
summary {
 systemName=NS_Pickaxe_FairyMochi_Trail
 emitters [{emitterName=Main_Trail, rendererClasses [&NiagaraRibbonRendererProperties]}]
 userVariables [
  User.RibbonIDOffset {type=NiagaraInt32}
  User.SwingSectionIndex {type=NiagaraInt32}
 ]
}
